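# Grassland ambience cue that plays randomized one-shot bird calls with outdoor bird attenuation and a distance-scaled send to distant reverb.
# A quality switch routes its first two slots to an even pick between a Blue Jay group and a mixed Chickadee, Finch, Flycatcher and Robin group.
# Its last two slots use a four-clip Chickadee, Finch and Flycatcher subset excluded from branch culling.
# Both branches vary pitch and volume through modulators.
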SoundCue /Game/Sounds/Artemis/Biomes/Grasslands/Grassland_Spots/Spot_Bird_Grassland_Classic_Calls_02_Cue
root=QualityLevel_0
nodes [
 {id=Modulator_1, children [Random_1], properties {pitchMax=1.1, volumeMax=1.1, volumeMin=0.9}}
 {id=Modulator_2, children [Random_3], properties {pitchMax=1.1, volumeMax=1.1, volumeMin=0.9}}
 {id=QualityLevel_0, children [Modulator_1,Modulator_1,Modulator_2,Modulator_2]}
 {id=Random_0, children [WavePlayer_1,WavePlayer_2,WavePlayer_3,WavePlayer_4,WavePlayer_6], properties {weights [1,1,1,1,1]}}
 {id=Random_1, children [Random_0,Random_2], properties {weights [1,1]}}
 {
  id=Random_2
  children [WavePlayer_7, WavePlayer_8, WavePlayer_9, WavePlayer_10, WavePlayer_11, WavePlayer_12, WavePlayer_13, WavePlayer_0]
  properties {weights [1,1,1,1,1,1,1,1]}
 }
 {
  id=Random_3
  children [WavePlayer_16, WavePlayer_17, WavePlayer_18, WavePlayer_21]
  properties {bShouldExcludeFromBranchCulling=true, weights [1,1,1,1]}
 }
 {id=WavePlayer_0, properties {soundWaveAssetPtr=&/Spot_Bird_Robin_Singing_01}}
 {id=WavePlayer_1, properties {soundWaveAssetPtr=&/Spot_Bird_BlueJay_01}}
 {id=WavePlayer_10, properties {soundWaveAssetPtr=&/Spot_Bird_Finch_01}}
 {id=WavePlayer_11, properties {soundWaveAssetPtr=&/Spot_Bird_Flycatcher_01}}
 {id=WavePlayer_12, properties {soundWaveAssetPtr=&/Spot_Bird_Flycatcher_02}}
 {id=WavePlayer_13, properties {soundWaveAssetPtr=&/Spot_Bird_Flycatcher_03}}
 {id=WavePlayer_16, properties {soundWaveAssetPtr=&/Spot_Bird_Chickadee_02}}
 {id=WavePlayer_17, properties {soundWaveAssetPtr=&/Spot_Bird_Chickadee_03}}
 {id=WavePlayer_18, properties {soundWaveAssetPtr=&Spot_Bird_Finch_01}}
 {id=WavePlayer_2, properties {soundWaveAssetPtr=&/Spot_Bird_BlueJay_02}}
 {id=WavePlayer_21, properties {soundWaveAssetPtr=&Spot_Bird_Flycatcher_03}}
 {id=WavePlayer_3, properties {soundWaveAssetPtr=&/Spot_Bird_BlueJay_03}}
 {id=WavePlayer_4, properties {soundWaveAssetPtr=&/Spot_Bird_BlueJay_04}}
 {id=WavePlayer_6, properties {soundWaveAssetPtr=&/Spot_Bird_BlueJay_06}}
 {id=WavePlayer_7, properties {soundWaveAssetPtr=&/Spot_Bird_Chickadee_01}}
 {id=WavePlayer_8, properties {soundWaveAssetPtr=&Spot_Bird_Chickadee_02}}
 {id=WavePlayer_9, properties {soundWaveAssetPtr=&Spot_Bird_Chickadee_03}}
]
properties {
 attenuationSettings=&Fort_TileAmbient_OneShot_Bird_Att
 bEnableBusSends=false
 maxDistance=8005
 soundClassObject=&Fort_Ambient_OneShots_Outdoors
 volumeMultiplier=0.301
 preEffectBusSends [{}]
 soundSubmixSends [{
   maxSendDistance=8000
   maxSendLevel=0.6
   minSendLevel=0.15
   sendLevelControlMethod=Linear
   soundSubmix=&Fortnite_Reverb_Distant
  }]
}
