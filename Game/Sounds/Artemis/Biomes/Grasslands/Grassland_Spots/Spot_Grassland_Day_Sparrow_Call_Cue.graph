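# Sound cue for daytime sparrow call one-shots, an ambient spot sound in the Grasslands biome.
# A quality-level node sends its first two slots to a random pick of five call waves and its last two to a smaller pool of three.
# Each pool passes through a modulator that varies pitch and volume.
# Output plays quietly on the outdoor ambient one-shot sound class with bird-specific attenuation.
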SoundCue /Game/Sounds/Artemis/Biomes/Grasslands/Grassland_Spots/Spot_Grassland_Day_Sparrow_Call_Cue
root=QualityLevel_0
nodes [
 {id=Modulator_1, children [Random_2], properties {pitchMax=1.1, volumeMax=1.1, volumeMin=0.9}}
 {id=Modulator_2, children [Random_3], properties {pitchMax=1.1, volumeMax=1.1, volumeMin=0.9}}
 {id=QualityLevel_0, children [Modulator_1,Modulator_1,Modulator_2,Modulator_2]}
 {id=Random_2, children [WavePlayer_9, WavePlayer_10, WavePlayer_11, WavePlayer_12, WavePlayer_13], properties {weights [1,1,1,1,1]}}
 {id=Random_3, children [WavePlayer_14, WavePlayer_15, WavePlayer_16], properties {weights [1,1,1]}}
 {id=WavePlayer_10, properties {soundWaveAssetPtr=&/Spot_Grassland_Day_Sparrow_Call_02}}
 {id=WavePlayer_11, properties {soundWaveAssetPtr=&/Spot_Grassland_Day_Sparrow_Call_03}}
 {id=WavePlayer_12, properties {soundWaveAssetPtr=&/Spot_Grassland_Day_Sparrow_Call_04}}
 {id=WavePlayer_13, properties {soundWaveAssetPtr=&/Spot_Grassland_Day_Sparrow_Call_05}}
 {id=WavePlayer_14, properties {soundWaveAssetPtr=&Spot_Grassland_Day_Sparrow_Call_03}}
 {id=WavePlayer_15, properties {soundWaveAssetPtr=&Spot_Grassland_Day_Sparrow_Call_04}}
 {id=WavePlayer_16, properties {soundWaveAssetPtr=&Spot_Grassland_Day_Sparrow_Call_05}}
 {id=WavePlayer_9, properties {soundWaveAssetPtr=&/Spot_Grassland_Day_Sparrow_Call_01}}
]
properties {
 attenuationSettings=&Fort_TileAmbient_OneShot_Bird_Att
 bEnableBusSends=false
 maxDistance=8005
 soundClassObject=&Fort_Ambient_OneShots_Outdoors
 volumeMultiplier=0.2
 preEffectBusSends [{}]
}
