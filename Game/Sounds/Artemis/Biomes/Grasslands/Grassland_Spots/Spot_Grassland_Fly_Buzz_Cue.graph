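# Sound cue for an outdoor grasslands ambient one-shot of a buzzing fly, played quietly with insect fly-by attenuation.
# A quality-level node sends its first two slots to a random pick of eight buzz waves and its last two to a smaller pool of four.
# Each pool passes through its own modulator, which varies pitch and volume slightly on each play.
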SoundCue /Game/Sounds/Artemis/Biomes/Grasslands/Grassland_Spots/Spot_Grassland_Fly_Buzz_Cue
root=QualityLevel_0
nodes [
 {id=Modulator_1, children [Random_2], properties {pitchMax=1.1, volumeMax=1, volumeMin=0.8}}
 {id=Modulator_2, children [Random_4], properties {pitchMax=1.1, volumeMax=1, volumeMin=0.8}}
 {id=QualityLevel_0, children [Modulator_1,Modulator_1,Modulator_2,Modulator_2]}
 {
  id=Random_2
  children [WavePlayer_8, WavePlayer_9, WavePlayer_10, WavePlayer_11, WavePlayer_12, WavePlayer_13, WavePlayer_14, WavePlayer_15]
  properties {weights [1,1,1,1,1,1,1,1]}
 }
 {id=Random_4, children [WavePlayer_16, WavePlayer_17, WavePlayer_18, WavePlayer_22], properties {weights [1,1,1,1]}}
 {id=WavePlayer_10, properties {soundWaveAssetPtr=&/Spot_Grassland_Fly_Buzz_03}}
 {id=WavePlayer_11, properties {soundWaveAssetPtr=&/Spot_Grassland_Fly_Buzz_04}}
 {id=WavePlayer_12, properties {soundWaveAssetPtr=&/Spot_Grassland_Fly_Buzz_05}}
 {id=WavePlayer_13, properties {soundWaveAssetPtr=&/Spot_Grassland_Fly_Buzz_06}}
 {id=WavePlayer_14, properties {soundWaveAssetPtr=&/Spot_Grassland_Fly_Buzz_07}}
 {id=WavePlayer_15, properties {soundWaveAssetPtr=&/Spot_Grassland_Fly_Buzz_08}}
 {id=WavePlayer_16, properties {soundWaveAssetPtr=&Spot_Grassland_Fly_Buzz_08}}
 {id=WavePlayer_17, properties {soundWaveAssetPtr=&Spot_Grassland_Fly_Buzz_06}}
 {id=WavePlayer_18, properties {soundWaveAssetPtr=&Spot_Grassland_Fly_Buzz_07}}
 {id=WavePlayer_22, properties {soundWaveAssetPtr=&Spot_Grassland_Fly_Buzz_03}}
 {id=WavePlayer_8, properties {soundWaveAssetPtr=&/Spot_Grassland_Fly_Buzz_01}}
 {id=WavePlayer_9, properties {soundWaveAssetPtr=&/Spot_Grassland_Fly_Buzz_02}}
]
properties {
 attenuationSettings=/Game/Sounds/Attenuation/Fort_TileAmbient_OneShot_Insect_By_Att
 bEnableBusSends=false
 maxDistance=6000
 soundClassObject=&Fort_Ambient_OneShots_Outdoors
 volumeMultiplier=0.15
 preEffectBusSends [{}]
}
