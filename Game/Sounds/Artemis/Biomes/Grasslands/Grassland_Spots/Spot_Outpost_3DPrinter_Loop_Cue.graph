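# Looping Sound Cue that plays intermittent 3D printer spot ambience for Grasslands outposts, in the Athena ambient props sound class.
# A Random node picks one of ten equally weighted waves, then a short random Delay and a pitch and volume Modulator space and vary each repeat.
# A Quality Level node feeds that chain to its first two slots only, leaving the other two silent.
# It reuses the outpost pipe attenuation and the Athena prop concurrency set.
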SoundCue /Game/Sounds/Artemis/Biomes/Grasslands/Grassland_Spots/Spot_Outpost_3DPrinter_Loop_Cue
root=Looping_0
nodes [
 {id=Delay_0, children [Random_0], properties {delayMax=5, delayMin=2}}
 {id=Looping_0, children [QualityLevel_0]}
 {id=Modulator_0, children [Delay_0], properties {pitchMax=1.2, volumeMax=1, volumeMin=0.5}}
 {id=QualityLevel_0, children [Modulator_0,Modulator_0,null,null]}
 {
  id=Random_0
  children [WavePlayer_1, WavePlayer_2, WavePlayer_3, WavePlayer_4, WavePlayer_5, WavePlayer_6, WavePlayer_7, WavePlayer_8, WavePlayer_9, WavePlayer_10]
  properties {bShouldExcludeFromBranchCulling=true, weights [1,1,1,1,1,1,1,1,1,1]}
 }
 {id=WavePlayer_1, properties {soundWaveAssetPtr=&/Spot_Outpost_3DPrinter_01}}
 {id=WavePlayer_10, properties {soundWaveAssetPtr=&/Spot_Outpost_3DPrinter_10}}
 {id=WavePlayer_2, properties {soundWaveAssetPtr=&/Spot_Outpost_3DPrinter_02}}
 {id=WavePlayer_3, properties {soundWaveAssetPtr=&/Spot_Outpost_3DPrinter_03}}
 {id=WavePlayer_4, properties {soundWaveAssetPtr=&/Spot_Outpost_3DPrinter_04}}
 {id=WavePlayer_5, properties {soundWaveAssetPtr=&/Spot_Outpost_3DPrinter_05}}
 {id=WavePlayer_6, properties {soundWaveAssetPtr=&/Spot_Outpost_3DPrinter_06}}
 {id=WavePlayer_7, properties {soundWaveAssetPtr=&/Spot_Outpost_3DPrinter_07}}
 {id=WavePlayer_8, properties {soundWaveAssetPtr=&/Spot_Outpost_3DPrinter_08}}
 {id=WavePlayer_9, properties {soundWaveAssetPtr=&/Spot_Outpost_3DPrinter_09}}
]
properties {
 attenuationSettings=&/Spot_Outpost_Pipe_Att
 bEnableBusSends=false
 maxDistance=3700
 soundClassObject=&Fort_Ambient_Props_Athena
 volumeMultiplier=1.253
 concurrencyOverrides {maxCount=2}
 concurrencySet [/Game/Sounds/Concurrency_Rules/Athena_Prop]
 preEffectBusSends [{}]
}
