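# Ambient Sound Cue for an outpost rift generator in the Grasslands biome, built from small terminal sounds and using desk attenuation and prop concurrency.
# A Mixer layers a pitched-down looping bed, shaped by an Enveloper volume fade-in, with randomly chosen chirps.
# The chirps repeat through a Looping node with random delays and wide pitch variation.
# A Quality Level node feeds that mix to its first two slots and leaves the others empty.
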SoundCue /Game/Sounds/Artemis/Biomes/Grasslands/Grassland_Spots/Spot_Outpost_Rift_Generator_Loop_Cue
root=QualityLevel_0
nodes [
 {id=Delay_1, children [Modulator_0], properties {delayMax=5, delayMin=1}}
 {
  id=Enveloper_0
  children [Modulator_1]
  properties {pitchCurve {editorCurveData {keys [{time=0}]}}, volumeCurve {editorCurveData {keys [<at 0 0>,<at 1 1>]}}}
 }
 {id=Looping_0, children [Delay_1]}
 {id=Mixer_0, children [Looping_0,Enveloper_0], properties {inputVolume [1,1]}}
 {id=Modulator_0, children [Random_0], properties {pitchMax=6, volumeMax=1, volumeMin=1}}
 {id=Modulator_1, children [WavePlayer_9], properties {pitchMax=0.6, pitchMin=0.5}}
 {id=QualityLevel_0, children [Mixer_0,Mixer_0,null,null]}
 {
  id=Random_0
  children [WavePlayer_1,WavePlayer_2,WavePlayer_3,WavePlayer_4,WavePlayer_5,WavePlayer_6,WavePlayer_7,WavePlayer_8]
  properties {bShouldExcludeFromBranchCulling=true, weights [1,1,1,1,1,1,1,1]}
 }
 {id=WavePlayer_1, properties {soundWaveAssetPtr=&/Spot_Outpost_Terminal_Small_Chirp_08}}
 {id=WavePlayer_2, properties {soundWaveAssetPtr=&/Spot_Outpost_Terminal_Small_Chirp_01}}
 {id=WavePlayer_3, properties {soundWaveAssetPtr=&/Spot_Outpost_Terminal_Small_Chirp_02}}
 {id=WavePlayer_4, properties {soundWaveAssetPtr=&/Spot_Outpost_Terminal_Small_Chirp_03}}
 {id=WavePlayer_5, properties {soundWaveAssetPtr=&/Spot_Outpost_Terminal_Small_Chirp_04}}
 {id=WavePlayer_6, properties {soundWaveAssetPtr=&/Spot_Outpost_Terminal_Small_Chirp_05}}
 {id=WavePlayer_7, properties {soundWaveAssetPtr=&/Spot_Outpost_Terminal_Small_Chirp_06}}
 {id=WavePlayer_8, properties {soundWaveAssetPtr=&/Spot_Outpost_Terminal_Small_Chirp_07}}
 {
  id=WavePlayer_9
  properties {bLooping=true, soundWaveAssetPtr=&/Spot_Outpost_Terminal_Small_Loop}
 }
]
properties {
 attenuationSettings=&/Spot_Outpost_Desk_Large_Att
 bEnableBusSends=false
 maxDistance=2550
 soundClassObject=&Fort_Ambient_Props_Athena
 volumeMultiplier=0.689
 concurrencyOverrides {maxCount=2}
 concurrencySet [/Game/Sounds/Concurrency_Rules/Athena_Prop]
 preEffectBusSends [{}]
}
