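# Sound cue for a small outpost terminal ambient spot in the Grasslands biome, attenuated and concurrency-limited as an Athena prop.
# A mixer layers a looping bed, shaped by an enveloper volume ramp, with repeating chirps.
# The chirp branch loops through a delay, randomizing volume and pitch on one of eight equally weighted waves.
# Quality-level nodes at the root and before the chirps fill only their first two slots.
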SoundCue /Game/Sounds/Artemis/Biomes/Grasslands/Grassland_Spots/Spot_Outpost_Terminal_Small_Loop_Cue
root=QualityLevel_1
nodes [
 {id=Delay_0, children [Modulator_1], properties {delayMax=6}}
 {
  id=Enveloper_0
  children [Modulator_0]
  properties {pitchCurve {editorCurveData {keys [{time=0}]}}, volumeCurve {editorCurveData {keys [<at 0 0>,<at 1 1>]}}}
 }
 {id=Looping_0, children [Delay_0]}
 {id=Mixer_0, children [Enveloper_0,Looping_0], properties {inputVolume [1,1]}}
 {id=Modulator_0, children [WavePlayer_0], properties {volumeMax=1, volumeMin=1}}
 {id=Modulator_1, children [QualityLevel_0], properties {pitchMax=2, volumeMin=0.2}}
 {id=QualityLevel_0, children [Random_1,Random_1,null,null]}
 {id=QualityLevel_1, children [Mixer_0,Mixer_0,null,null]}
 {
  id=Random_1
  children [WavePlayer_10, WavePlayer_11, WavePlayer_12, WavePlayer_13, WavePlayer_14, WavePlayer_15, WavePlayer_16, WavePlayer_17]
  properties {weights [1,1,1,1,1,1,1,1]}
 }
 {
  id=WavePlayer_0
  properties {bLooping=true, soundWaveAssetPtr=&/Spot_Outpost_Terminal_Small_Loop}
 }
 {id=WavePlayer_10, properties {soundWaveAssetPtr=&/Spot_Outpost_Terminal_Small_Chirp_01}}
 {id=WavePlayer_11, properties {soundWaveAssetPtr=&/Spot_Outpost_Terminal_Small_Chirp_02}}
 {id=WavePlayer_12, properties {soundWaveAssetPtr=&/Spot_Outpost_Terminal_Small_Chirp_03}}
 {id=WavePlayer_13, properties {soundWaveAssetPtr=&/Spot_Outpost_Terminal_Small_Chirp_04}}
 {id=WavePlayer_14, properties {soundWaveAssetPtr=&/Spot_Outpost_Terminal_Small_Chirp_05}}
 {id=WavePlayer_15, properties {soundWaveAssetPtr=&/Spot_Outpost_Terminal_Small_Chirp_06}}
 {id=WavePlayer_16, properties {soundWaveAssetPtr=&/Spot_Outpost_Terminal_Small_Chirp_07}}
 {id=WavePlayer_17, properties {soundWaveAssetPtr=&/Spot_Outpost_Terminal_Small_Chirp_08}}
]
properties {
 attenuationSettings=&/Spot_Outpost_Terminal_Small_Att
 bEnableBusSends=false
 maxDistance=3650
 soundClassObject=&Fort_Ambient_Props_Athena
 volumeMultiplier=1.047
 concurrencyOverrides {maxCount=2}
 concurrencySet [/Game/Sounds/Concurrency_Rules/Athena_Prop]
 preEffectBusSends [{}]
}
