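# Looping ambient Sound Cue for an outpost water pump in the Grasslands biome, on the Athena ambient props sound class.
# A looping wave player runs through a Modulator with fixed volume into a Quality Level node, which connects only the first two slots and leaves the others empty.
# A root Enveloper fades the volume in.
# Pipe attenuation sets its spatial falloff, and prop concurrency rules apply.
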SoundCue /Game/Sounds/Artemis/Biomes/Grasslands/Grassland_Spots/Spot_Outpost_WaterPump_Loop_Cue
root=Enveloper_0
nodes [
 {
  id=Enveloper_0
  children [QualityLevel_0]
  properties {pitchCurve {editorCurveData {keys [{time=0}]}}, volumeCurve {editorCurveData {keys [<at 0 0>,<at 1 1>]}}}
 }
 {id=Modulator_0, children [WavePlayer_0], properties {volumeMax=1, volumeMin=1}}
 {id=QualityLevel_0, children [Modulator_0,Modulator_0,null,null]}
 {
  id=WavePlayer_0
  properties {bLooping=true, soundWaveAssetPtr=&/Spot_Outpost_WaterPump_Loop_01}
 }
]
properties {
 attenuationSettings=&/Spot_Outpost_Pipe_Att
 bEnableBusSends=false
 maxDistance=3700
 soundClassObject=&Fort_Ambient_Props_Athena
 volumeMultiplier=1.504
 concurrencyOverrides {maxCount=2}
 concurrencySet [/Game/Sounds/Concurrency_Rules/Athena_Prop]
 preEffectBusSends [{}]
}
