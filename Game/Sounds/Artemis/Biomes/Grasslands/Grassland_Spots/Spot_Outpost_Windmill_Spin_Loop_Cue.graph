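# Looping Sound Cue for the spinning outpost windmill in the Grasslands biome, classed as an ambient prop.
# A Mixer layers two looping waves: a distant loop with mono attenuation and a slightly quieter close loop with stereo attenuation, so the layers follow listener distance.
# Each layer runs through a fixed-value Modulator and an Enveloper that fades volume in from silence.
# The cue uses prop concurrency rules.
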SoundCue /Game/Sounds/Artemis/Biomes/Grasslands/Grassland_Spots/Spot_Outpost_Windmill_Spin_Loop_Cue
root=Mixer_1
nodes [
 {
  id=Attenuation_1
  children [Enveloper_0]
  properties {attenuationSettings=&/Spot_Outpost_Windmill_Distant_Mono_Att}
 }
 {
  id=Attenuation_2
  children [Enveloper_1]
  properties {attenuationSettings=&/Spot_Outpost_Windmill_Close_Stereo_Att}
 }
 {
  id=Enveloper_0
  children [Modulator_0]
  properties {pitchCurve {editorCurveData {keys [{time=0}]}}, volumeCurve {editorCurveData {keys [<at 0 0>,<at 1 1>]}}}
 }
 {
  id=Enveloper_1
  children [Modulator_1]
  properties {pitchCurve {editorCurveData {keys [{time=0}]}}, volumeCurve {editorCurveData {keys [<at 0 0>,<at 1 1>]}}}
 }
 {id=Mixer_1, children [Attenuation_1, Attenuation_2], properties {inputVolume [1,0.8]}}
 {id=Modulator_0, children [WavePlayer_0], properties {pitchMax=1, pitchMin=1, volumeMax=1, volumeMin=1}}
 {id=Modulator_1, children [WavePlayer_1], properties {pitchMax=1, pitchMin=1, volumeMax=1, volumeMin=1}}
 {
  id=WavePlayer_0
  properties {bLooping=true, soundWaveAssetPtr=&/Spot_Outpost_Windmill_Spin_Loop}
 }
 {
  id=WavePlayer_1
  properties {bLooping=true, soundWaveAssetPtr=&/Spot_Outpost_Windmill_Spin_Close_Loop}
 }
]
properties {
 attenuationSettings=&Spot_Outpost_Windmill_Distant_Mono_Att
 bEnableBusSends=false
 maxDistance=3250
 soundClassObject=&Fort_Ambient_Props_Athena
 volumeMultiplier=1.504
 concurrencyOverrides {maxCount=2}
 concurrencySet [/Game/Sounds/Concurrency_Rules/Athena_Prop]
 preEffectBusSends [{}]
}
